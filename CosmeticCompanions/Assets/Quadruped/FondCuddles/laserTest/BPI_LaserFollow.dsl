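; Blueprint interface for laser-following behaviour, in a quadruped cosmetic companion's laser test folder.
; It declares queries for whether the implementer is following a laser, whether the laser is moving, and the laser player reference.
; Start and stop calls control following, with start taking a player actor reference.
; The functions carry no logic, only default returns.
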(fn GetIsFollowingLaser
  (return false))

(fn GetIsLaserMoving
  (return false))

(fn GetLaserPlayerRef
  (return 0))

(fn StartFollowingLaser
  (params (PlayerRef :type Actor%20Object%20Reference)))

(fn StopFollowingLaser)
