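# Test AnimBP for the FondCuddles quadruped companion, switching between laser-chase walk/idle sequences and poses retargeted from another mesh.
# Normal retargeting uses one of two IK retargeters, one adding a heart attachment constraint.
# Each update polls the companion's BPI_LaserFollow state and calls StartFollowingLaser when LaserFollowCurve rises.
# Look-at bones, trace-based movement, rotation and the timed stop are present but unwired.
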AnimBlueprint /CosmeticCompanions/Assets/Quadruped/FondCuddles/laserTest/FondCuddles_Quadruped_AnimBP1
target_skeleton=/CosmeticCompanions/Assets/Quadruped/FondCuddles/Meshes/FondCuddles_Quadruped_Skeleton
graphs [
 AnimGraph {nodes [
   n1 = Look At - Bone: spine_02!(ComponentPose=Local To Component(), Alpha=1, LookAtLocation:Vector="100,0,0") -> (Pose) {
    interpolationTime=2
    interpolationTriggerThreashold=10
    boneToModify {boneName=spine_02}
    lookAt_Axis {bInLocalSpace=false, axis {y=0}}
    lookUp_Axis {bInLocalSpace=false}
    lookAtTarget=~
   }
   n2 = Look At - Bone: neck_02!(ComponentPose=@n1, Alpha=1, LookAtLocation:Vector="100,0,0") -> (Pose) {
    interpolationTime=1
    interpolationTriggerThreashold=10
    boneToModify {boneName=neck_02}
    lookAt_Axis {bInLocalSpace=false, axis {y=0}}
    lookUp_Axis {bInLocalSpace=false}
    lookAtTarget=~
   }
   Component To Local(ComponentPose=@n2)
   Copy Pose From Mesh() {bCopyCurves=true, bUseAttachedParent=true}
   Retarget Pose From Mesh!() -> (Pose:Pose Link Structure) {is=title AnimGraphNode_RetargetPoseFromMesh}
   n3 = Sequence Player!(Sequence:Anim Sequence Base Object Reference=Get LaserWalkAnim() -> (LaserWalkAnim:Anim Sequence Object Reference)) -> (Pose)
   n4 = Sequence Player(Sequence=Get LaserIdleAnim() -> (LaserIdleAnim:Anim Sequence Object Reference))
   n5 = Get bIsLaserMoving() -> (bIsLaserMoving:bool)
   n6 = "Constraint - attach_emote_01 (P)"!(
    ComponentPose=Local To Component(LocalPose=Retarget Pose From Mesh() {iKRetargeterAsset=/CosmeticCompanions/Assets/Quadruped/FondCuddles/Meshes/RTG_FondCuddles})
    bAlphaBoolEnabled:bool=NOT Boolean(A=Get bUseReverseRetarget() -> (bUseReverseRetarget:bool))
    ConstraintWeights_0=1
   ) -> (Pose) {alphaInputType=Bool, boneToModify {boneName=attach_emote_01}, constraintSetup [{targetBone {boneName=dyn_heart_01}}], constraintWeights [1]}
   n7 = Component To Local(ComponentPose=@n6)
   n8 = Retarget Pose From Mesh() {iKRetargeterAsset=/CosmeticCompanions/Assets/Quadruped/FondCuddles/Meshes/RTG_FondCuddles_Emote}
   n9 = Get bUseReverseRetarget()
   Output Pose(Result=Blend Poses by bool(
     BlendPose_0=Blend Poses by bool(BlendPose_0=@n3, BlendPose_1=@n4, BlendTime_0=0.2, BlendTime_1=0.2, bActiveValue=@n5)
     BlendPose_1=Blend Poses by bool(BlendPose_0=@n8, BlendPose_1=@n7, bActiveValue=@n9)
     bActiveValue=Get bIsFollowingLaser() -> (bIsFollowingLaser:bool)
    ))
   Select!(
    Option 0:Anim Sequence Object Reference=/BRCosmetics/Animation/Game/MainPlayer/Skydive/ParaGlide/SharkLaser/SharkLaser_Cannon_Shoot_GLIDER
    Option 1:Anim Sequence Object Reference=/BRCosmetics/Animation/Game/MainPlayer/Skydive/ParaGlide/SharkLaser/SharkLaser_Cannon_Shoot_2x_GLIDER
    Option 2:Anim Sequence Object Reference=/BRCosmetics/Animation/Game/MainPlayer/Skydive/ParaGlide/SharkLaser/SharkLaser_Cannon_Shoot_3x_GLIDER
    Index:int=0
   ) -> (ReturnValue:Anim Sequence Object Reference)
  ]}
 EventGraph {nodes [
   n1 = FindLookAtRotation!(
    Start:"Vector (by ref)"=Get World Location!(self:Scene Component Object Reference=Get MeshComp() -> (MeshComp:Skeletal Mesh Component Object Reference)) -> (ReturnValue:Vector) {is=title K2Node_CallFunction}
    Target:"Vector (by ref)"=Get LaserTargetLocation() -> (LaserTargetLocation:Vector)
   ) -> (ReturnValue:Rotator)
   n2 = RInterpTo!(
    Current:Rotator=Get World Rotation!(self:Scene Component Object Reference=Get MeshComp()) -> (ReturnValue:Rotator) {is=title K2Node_CallFunction}
    Target:Rotator=@n1
    DeltaTime:float=Get DeltaTime() -> (DeltaTime:double)
    InterpSpeed:float=10
   ) -> (ReturnValue:Rotator)
   n3 = Get MeshComp()
   n4 = Set World Rotation!(execute, self:Scene Component Object Reference=@n3, NewRotation:Rotator=@n2, bSweep:bool=false, bTeleport:bool=false) -> (then, SweepHitResult:Hit Result Structure) {is=title K2Node_CallFunction}
   n5 = Get World Location(self=Get MeshComp())
   n6 = Get LaserTargetLocation()
   Branch(
    execute=[@n4.then, @n7.then]
    Condition="float < float"(A="Distance2D (Vector)"!(V1:Vector=@n5, V2:Vector=@n6) -> (ReturnValue:double), B=50)
   )
   > Set bHasLaserTarget(bHasLaserTarget:bool=false) -> (Output_Get:bool)
   n7 = Set Relative Rotation!(execute, self:Scene Component Object Reference=@n3, NewRotation:Rotator=@n2, bSweep:bool=false, bTeleport:bool=false) -> (then, SweepHitResult:Hit Result Structure) {is=title K2Node_CallFunction}
   n8 = Event BlueprintUpdateAnimation()
   > Set DeltaTime(DeltaTime:double=@n8.DeltaTimeX) -> (Output_Get:double)
   > n9 = Branch(Condition=IsValid(Object=Get CompanionRef() -> (CompanionRef:Actor Object Reference)))
   n10 = Cast To BPI_LaserFollow!(execute=[@n9.then, @n17.then], Object=Get CompanionRef()) -> (then, CastFailed, AsBPI Laser Follow:BPI Laser Follow Interface) {is=title K2Node_DynamicCast}
   > n11 = GetIsFollowingLaser!(execute, self:BPI Laser Follow Interface=@n10) -> (then, ReturnValue:bool)
   > Set bIsFollowingLaser(bIsFollowingLaser:bool=@n11) -> (Output_Get:bool)
   > n12 = GetIsLaserMoving!(execute:Exec, self:Object Reference=Get CompanionRef()) -> (then:Exec, ReturnValue:bool) {is=title K2Node_Message}
   > Set bIsLaserMoving(bIsLaserMoving:bool=@n12) -> (Output_Get:bool)
   > n13 = Branch(Condition=Get bLaserCurveWasActive() -> (bLaserCurveWasActive:bool))
   > n14 = Branch(Condition=Get bIsFollowingLaser())
   Set bLaserCurveWasActive(execute=@n14.else, bLaserCurveWasActive:bool=false) -> (Output_Get:bool)
   n15 = GetCurveValue(CurveName=LaserFollowCurve)
   Branch(execute=@n13.else, Condition="float > float"(A=@n15, B=0.1))
   > Set bLaserCurveWasActive(bLaserCurveWasActive=true)
   > StartFollowingLaser!(
    execute:Exec
    self:Object Reference=Get CompanionRef()
    PlayerRef:Actor Object Reference=GetPlayerPawn!(PlayerIndex:int=0) -> (ReturnValue:Pawn Object Reference)
   ) -> (then:Exec) {is=title K2Node_Message}
   n16 = GetOwningComponent()
   Set CompanionRef(execute=@n9.else, CompanionRef:Actor Object Reference=GetOwner(self=@n16)) -> (Output_Get:Actor Object Reference)
   > n17 = Set MeshComp(MeshComp:Skeletal Mesh Component Object Reference=@n16) -> (Output_Get:Skeletal Mesh Component Object Reference)
   n18 = Branch()
   > Set LaserTargetLocation(LaserTargetLocation:Vector=BreakHitResult!(Hit:"Hit Result Structure (by ref)") -> (
     bBlockingHit:bool
     bInitialOverlap:bool
     Time:float
     Distance:float
     Location:Vector
     ImpactPoint:Vector
     Normal:Vector
     ImpactNormal:Vector
     PhysMat:Physical Material Object Reference
     HitActor:Actor Object Reference
     HitComponent:Primitive Component Object Reference
     HitBoneName:Name
     BoneName:Name
     HitItem:int
     ElementIndex:int
     FaceIndex:int
     TraceStart:Vector
     TraceEnd:Vector
    ).ImpactPoint) -> (Output_Get:Vector)
   > n19 = Set bHasLaserTarget(bHasLaserTarget=true)
   n20 = Branch(execute=[@n33.then, @n19.then], Condition=Get bHasLaserTarget() -> (bHasLaserTarget:bool))
   n21 = Get World Location(self=Get MeshComp())
   n22 = Get LaserTargetLocation()
   > n23 = Branch(Condition="float > float"(A="Distance2D (Vector)"(V1=@n21, V2=@n22), B=100))
   > n24 = Branch(Condition=NOT Boolean(A=Get bIsLaserMoving()))
   > n25 = Set bIsLaserMoving(bIsLaserMoving=true)
   n26 = Get MeshComp()
   n27 = BreakVector(InVec=vector - vector(A=Get LaserTargetLocation(), B=Get World Location(self=@n26)))
   n28 = "Normalize 2D (Vector)"!(A:Vector=MakeVector(X=@n27, Y=@n27.Y, Z=@n27.Z), Tolerance:float=0.0001) -> (ReturnValue:Vector)
   n29 = Get LaserMoveSpeed() -> (LaserMoveSpeed:double)
   n30 = vector + vector(
    A=vector * vector(A=vector * vector(A=@n28, B=@n29) -> (ReturnValue=~), B=Get DeltaTime()) -> (ReturnValue=~)
    B=Get World Location(self=Get MeshComp())
   )
   n31 = BreakVector(InVec=@n30)
   n32 = Line Trace By Channel!(
    execute=[@n24.else, @n25.then]
    Start:Vector=MakeVector(X=@n31, Y=@n31.Y, Z=float + float(A=@n31.Z, B=200))
    End:Vector=MakeVector(X=@n31, Y=@n31.Y, Z=float - float(A=@n31.Z, B=500))
    TraceChannel:ETraceTypeQuery Enum
    bTraceComplex:bool=false
    ActorsToIgnore:Array of Actor Object References=Make Array("[0]":Actor Object Reference=Get CompanionRef()) -> (Array:Array of Actor Object References)
    DrawDebugType:EDrawDebugTrace Enum=ForDuration
    bIgnoreSelf:bool
    TraceColor:Linear Color Structure
    TraceHitColor:Linear Color Structure
    DrawTime:float
   ) -> (then, OutHit:Hit Result Structure, ReturnValue:bool)
   > Branch(Condition=@n32.ReturnValue)
   Branch(execute=@n23.else, Condition=Get bIsLaserMoving())
   > Set bIsLaserMoving(bIsLaserMoving=false)
   Branch(execute=@n20.else, Condition=Get bIsLaserMoving())
   > Set bIsLaserMoving(bIsLaserMoving=false)
   PrintString!(execute=@n18.else, InString=MISS, bPrintToScreen, bPrintToLog, TextColor, Duration, Key) -> (then)
   > n33 = Set bHasLaserTarget(bHasLaserTarget=false)
   n34 = Branch(Condition="float >= float"(A=GetCurveValue(CurveName=LaserFollowCurve), B=0.1))
   > Set DeactivateTimer(DeactivateTimer:double=0) -> (Output_Get:double)
   n35 = Set DeactivateTimer(execute=@n34.else, DeactivateTimer=float + float(A=Get DeltaTime(), B=Get DeactivateTimer() -> (DeactivateTimer:double)))
   > Branch(Condition="float > float"(A=@n35, B=2))
   > Set bLaserCurveWasActive(bLaserCurveWasActive=false)
   > StopFollowingLaser!(execute:Exec, self:Object Reference=Get CompanionRef()) -> (then:Exec) {is=title K2Node_Message}
   n36 = Get HeadYaw!(self:Laser Follow Component Object Reference=GetComponentByClass!(
     self:Actor Object Reference=Get CompanionRef()
     ComponentClass:Actor Component Class Reference=&/LaserFollowComponent.LaserFollowComponent_C
    ) -> (ReturnValue:Laser Follow Component Object Reference)) -> (HeadYaw:double)
   Set HeadYaw(HeadYaw:double=@n36) -> (Output_Get:double)
   > PrintString(InString="To String (Float)"(InDouble=@n36))
   Get Actor Location(self=Get CompanionRef())
   Get Actor Location(self=Get CompanionRef())
   Set World Location!(execute, self:Scene Component Object Reference=Get MeshComp(), NewLocation:Vector="0, 0, 0", bSweep:bool=true, bTeleport:bool=false) -> (then, SweepHitResult:Hit Result Structure) {is=title K2Node_CallFunction}
   Set Actor Location!(execute, self:Actor Object Reference=Get CompanionRef(), NewLocation:Vector="0, 0, 0", bSweep:bool=true, bTeleport:bool=true) -> (then, SweepHitResult:Hit Result Structure, ReturnValue:bool) {is=title K2Node_CallFunction}
   Set Actor Location(NewLocation="0, 0, 0", bSweep=true, bTeleport=false)
   Set World Location(self=Get MeshComp(), NewLocation="0, 0, 0", bSweep=true, bTeleport=false)
   Set Actor Rotation!(execute, self:Actor Object Reference, NewRotation:Rotator="0, 0, 0", bTeleportPhysics:bool=false) -> (then, ReturnValue:bool) {is=title K2Node_CallFunction}
   Get Actor Rotation!(self:Actor Object Reference=Get CompanionRef()) -> (ReturnValue:Rotator) {is=title K2Node_CallFunction}
   Get Actor Location(self=Get CompanionRef())
   Get Actor Location(self=Get CompanionRef())
   Set Actor Location(self=Get CompanionRef(), NewLocation="0, 0, 0", bSweep=true, bTeleport=false)
   Set Actor Location(self=Get CompanionRef(), NewLocation="0, 0, 0", bSweep=true, bTeleport=true)
   "To String (Float)"(InDouble=BreakRotator(InRot=@n1).Yaw)
   Get Actor Location(self=Get CompanionRef())
   BreakHitResult(Hit=@n32)
   Reroute Node(InputPin:Vector=@n30) -> (OutputPin:Vector)
   "float >= float"(A=@n15, B=0.5)
   Get CompanionRef()
   Get CompanionRef()
   Get CompanionRef()
  ]}
]
variables [
 bHasLaserTarget {type=boolean}
 bIsFollowingLaser {type=boolean}
 bIsLaserMoving {type=boolean}
 bLaserCurveWasActive {type=boolean}
 CompanionRef {type=Actor, default=null}
 DeactivateTimer {type=number}
 DeltaTime {type=number}
 HeadYaw {type=number}
 isLaserMoving {type=boolean}
 LaserIdleAnim {type=AnimSequence, default=&/FlourCut_Angry_Idle_Loop_LaserTest}
 LaserMoveSpeed {type=number, default=400}
 LaserPlayerRef {type=Actor, default=null}
 LaserTarget {type=Vector}
 LaserTargetLocation {type=Vector}
 LaserTraceRange {type=number, default=5000}
 LaserWalkAnim {type=AnimSequence, default=&/FlourCut_Walk_LaserTest}
 MeshComp {type=SkeletalMeshComponent}
 Target {default=None}
]
