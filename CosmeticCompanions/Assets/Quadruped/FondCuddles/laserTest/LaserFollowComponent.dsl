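; Actor component that makes a companion creature chase a laser dot aimed from a Fortnite player's view.
; Activation spawns a dot decal, and Tick line-traces from that view to place it and set the chase target.
; Movement scales speed by distance, checks for obstacles and range from the start, then ground-snaps actor and mesh with smoothed slope-aligned rotation.
; Following deactivates after a trace-miss timeout or when the LaserFollowCurve anim curve stays low briefly.
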(event EventBeginPlay
  (Development|PrintString (Utilities|GetDisplayName (Components|GetOwner)))
  (Variables|Default|SetOwnerActor (Components|GetOwner))
  (Variables|Default|SetOwnerMesh (Actor|GetComponentbyClass (Variables|Default|GetOwnerActor) "/Script/Engine.SkeletalMeshComponent")))

(event EventTick
  (params (DeltaSeconds :type Float%20%28single-precision%29))
  (bind _deltatime (Variables|Default|GetDeltaTime))
  (bind _curvedroptimer (Variables|Default|GetCurveDropTimer))
  (bind _returnvalue (+ _curvedroptimer _deltatime))
  (Variables|Default|SetDeltaTime DeltaSeconds)
  (if (and (Variables|Default|GetbIsFollowingLaser) (Utilities|IsValid (Variables|Default|GetLaserPlayerRef) :!class /Script/Engine.KismetSystemLibrary))
    (if (< (Animation|Curves|GetCurveValue (Components|SkeletalMesh|GetAnimInstance (Variables|Default|GetOwnerMesh)) "LaserFollowCurve") 0.1)
      (Variables|Default|SetCurveDropTimer _returnvalue) ; Accumulate
      (if (> _returnvalue 0.1)
        (CallFunction|DeactivateLaserFollow)
        (else
          _))
      (else
        (Variables|Default|SetCurveDropTimer 0.0)))
    (bind _asfort_player_pawn (Utilities|Casting|CastToFortPlayerPawn (Variables|Default|GetLaserPlayerRef))
      (:then
        (bind (_outhit _returnvalue_2) (Collision|LineTraceByChannel (Pawn|GetPlayerViewPoint (Pawn|GetController _asfort_player_pawn)) (+ (* (Math|Vector|GetForwardVector (.Rotation (Pawn|GetPlayerViewPoint (Pawn|GetController _asfort_player_pawn)))) (Variables|Default|GetLaserTraceRange)) (Pawn|GetPlayerViewPoint (Pawn|GetController _asfort_player_pawn))) "TraceTypeQuery1" false (Utilities|Array|MakeArray (Variables|Default|GetLaserPlayerRef) (Variables|Default|GetOwnerActor))))
        (if _returnvalue_2
          ; lazer
          (bind _output_get (Variables|Default|SetLaserTargetLocation (.ImpactPoint (Collision|BreakHitResult _outhit))))
          (Transformation|SetWorldLocationAndRotation (Variables|Default|GetLaserDotDecal) _output_get (Math|Rotator|MakeRotfromZ (.ImpactNormal (Collision|BreakHitResult _outhit))) false false)
          (Rendering|Components|Decal|SetFadeScreenSize (Variables|Default|GetLaserDotDecal) 0.001)
          (Variables|Default|SetbHasLaserTarget true)
          (else
            (Variables|Default|SetbHasLaserTarget false)
            (Variables|Default|SetTraceMissTimer (+ (Variables|Default|GetDeltaTime) (Variables|Default|GetTraceMissTimer)))
            (if (> (Variables|Default|GetTraceMissTimer) (Variables|Default|GetTraceMissTimeout))
              (Development|PrintString "TIMEOUT DEACTIVATE")
              (CallFunction|DeactivateLaserFollow)
              (elif (Variables|Default|GetbHasLaserTarget)
                (Variables|Default|SetTraceMissTimer 0.0)
                (if (> (Math|Vector|Distance2D(Vector) (Variables|Default|GetLastSetPosition) (Variables|Default|GetLaserTargetLocation)) 4e+01)
                  (Variables|Default|SetbIsLaserMoving true)
                  (bind (_outhit_2 _returnvalue_3) (Collision|LineTraceByChannel (Math|Vector|MakeVector (.x (Variables|Default|GetLastSetPosition)) (.y (Variables|Default|GetLastSetPosition)) (+ (.z (Variables|Default|GetLastSetPosition)) 4e+01)) (Math|Vector|MakeVector (.x (Variables|Default|GetLaserTargetLocation)) (.y (Variables|Default|GetLaserTargetLocation)) (+ (.z (Variables|Default|GetLastSetPosition)) 4e+01)) "TraceTypeQuery1" false (Utilities|Array|MakeArray (Variables|Default|GetOwnerActor))))
                  (if _returnvalue_3
                    (Variables|Default|SetbIsLaserMoving false)
                    (else
                      (bind _output_get_2 (Variables|Default|SetNewLocation (+ (* (* (Math|Vector|Normalize2D(Vector) (Math|Vector|MakeVector (.x (- (Variables|Default|GetLaserTargetLocation) (Transformation|GetActorLocation (Variables|Default|GetOwnerActor)))) (.y (- (Variables|Default|GetLaserTargetLocation) (Transformation|GetActorLocation (Variables|Default|GetOwnerActor)))))) (* (Math|Float|Clamp(Float) (/ (Math|Vector|Distance2D(Vector) (Transformation|GetActorLocation (Variables|Default|GetOwnerActor)) (Variables|Default|GetLaserTargetLocation)) (Variables|Default|GetRampDistance)) (Variables|Default|GetMinSpeedFraction)) (Variables|Default|GetLaserMoveSpeed))) (Variables|Default|GetDeltaTime)) (Transformation|GetActorLocation (Variables|Default|GetOwnerActor)))))
                      (Transformation|SetActorLocation (Variables|Default|GetOwnerActor) _output_get_2 false false)))
                  ; distance check
                  (if (> (Math|Vector2D|Distance2D (Math|Conversions|ToVector2D(Vector) (Variables|Default|GetNewLocation)) (Math|Conversions|ToVector2D(Vector) (Variables|Default|GetLaserStartPosition))) (Variables|Default|GetLaserMaxRange))
                    (Variables|Default|SetbIsLaserMoving false)
                    (Transformation|SetWorldLocation (Variables|Default|GetOwnerMesh) (Variables|Default|GetLastSetPosition) false false)
                    (else
                      (Transformation|SetWorldLocation (Variables|Default|GetOwnerMesh) (Variables|Default|GetNewLocation) false false)
                      (Variables|Default|SetLastSetPosition (Variables|Default|GetNewLocation))
                      (Variables|Default|SetCurveDropTimer 0.0)
                      (if (< (Math|Vector|Distance2D(Vector) (Variables|Default|GetLastSetPosition) (Variables|Default|GetLaserTargetLocation)) 5e+01)
                        (Variables|Default|SetbHasLaserTarget false))))
                  (bind (_outhit_3 _returnvalue_4) (Collision|LineTraceByChannel (Math|Vector|MakeVector (.x (Variables|Default|GetLastSetPosition)) (.y (Variables|Default|GetLastSetPosition)) (+ (.z (Variables|Default|GetLastSetPosition)) 2e+02)) (Math|Vector|MakeVector (.x (Math|Vector|MakeVector (.x (Variables|Default|GetLastSetPosition)) (.y (Variables|Default|GetLastSetPosition)) (+ (.z (Variables|Default|GetLastSetPosition)) 2e+02))) (.y (Math|Vector|MakeVector (.x (Variables|Default|GetLastSetPosition)) (.y (Variables|Default|GetLastSetPosition)) (+ (.z (Variables|Default|GetLastSetPosition)) 2e+02))) (- (.z (Math|Vector|MakeVector (.x (Variables|Default|GetLastSetPosition)) (.y (Variables|Default|GetLastSetPosition)) (+ (.z (Variables|Default|GetLastSetPosition)) 2e+02))) 1.5e+03))))
                  ; ground trace
                  (if _returnvalue_4
                    (Transformation|SetWorldLocation (Variables|Default|GetOwnerMesh) (Math|Vector|MakeVector (.x (.ImpactPoint (Collision|BreakHitResult _outhit_3))) (.y (.ImpactPoint (Collision|BreakHitResult _outhit_3))) (.z (.ImpactPoint (Collision|BreakHitResult _outhit_3)))) false false)
                    (bind _output_get_3 (Variables|Default|SetLastSetPosition (Math|Vector|MakeVector (.x (.ImpactPoint (Collision|BreakHitResult _outhit_3))) (.y (.ImpactPoint (Collision|BreakHitResult _outhit_3))) (.z (.ImpactPoint (Collision|BreakHitResult _outhit_3))))))
                    (Transformation|SetActorLocation (Variables|Default|GetOwnerActor) _output_get_3 false false))
                  (bind _output_get_4 (Variables|Default|SetSmoothedRotation (Math|Interpolation|RInterpTo (Variables|Default|GetSmoothedRotation) (Math|Rotator|MakeRotator (.roll (Math|Rotator|MakeRotfromZX (.ImpactNormal (Collision|BreakHitResult _outhit_3)) (Transformation|GetForwardVector (Variables|Default|GetOwnerMesh)))) (.pitch (Math|Rotator|MakeRotfromZX (.ImpactNormal (Collision|BreakHitResult _outhit_3)) (Transformation|GetForwardVector (Variables|Default|GetOwnerMesh)))) (- (.yaw (Math|Rotator|FindLookatRotation (Variables|Default|GetLastSetPosition) (Variables|Default|GetLaserTargetLocation))) 9e+01)) (Variables|Default|GetDeltaTime) 15.0)))
                  (Transformation|SetWorldRotation (Variables|Default|GetOwnerMesh) _output_get_4 false false)
                  (else
                    (Variables|Default|SetbIsLaserMoving false)
                    (Variables|Default|SetbHasLaserTarget false)
                    (Transformation|SetWorldLocation (Variables|Default|GetOwnerMesh) (Variables|Default|GetLastSetPosition) false false)))
                (else
                  _))))))
      (:CastFailed
        (Development|PrintString "cast failed")))))

(event Custom|ActivateLaserFollow
  (params (PlayerRef :type Actor%20Object%20Reference))
  (Variables|Default|SetLaserPlayerRef PlayerRef)
  (Variables|Default|SetbIsFollowingLaser true)
  (Variables|Default|SetLaserStartPosition (Transformation|GetWorldLocation (Variables|Default|GetOwnerMesh)))
  (bind _output_get (Variables|Default|SetLastSetPosition (Transformation|GetWorldLocation (Variables|Default|GetOwnerMesh))))
  (bind _returnvalue (Rendering|Decal|SpawnDecalatLocation "/CosmeticCompanions/Assets/Quadruped/FondCuddles/laserTest/CatLaserTest_Player/LaserDot.LaserDot" "2.500000,2.500000,2.500000" _output_get "0.000000,0.000000,0.000000"))
  (Variables|Default|SetLaserDotDecal _returnvalue)
  (Variables|Default|SetPreviousLocation (Transformation|GetActorLocation (Variables|Default|GetOwnerActor)))
  (Variables|Default|SetSmoothedRotation (Transformation|GetWorldRotation (Variables|Default|GetOwnerMesh))))

(event Custom|DeactivateLaserFollow
  (Variables|Default|SetbIsFollowingLaser false)
  (Variables|Default|SetbHasLaserTarget false)
  (Variables|Default|SetbIsLaserMoving false)
  (Variables|Default|SetLaserPlayerRef 0)
  (Variables|Default|SetTraceMissTimer 0.0)
  (Variables|Default|SetCurveDropTimer 0.0)
  (Components|DestroyComponent (Variables|Default|GetLaserDotDecal))
  (Variables|Default|SetLaserDotDecal 0)
  (Actor|Tick|SetActorTickEnabled (Variables|Default|GetOwnerActor) true))
